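# Translucent UI material for a skewed button in the frontend spatial UI, drawn procedurally without textures.
# Rotated, offset UV edges with adjustable feather form quad masks, using normalized UV for top and bottom and 9-slice UV for sides.
# Two masks have separate button and reflection opacities, the button's shifted by an animatable reflection offset, and are combined by add and multiply.
# Color is RGB1 or, via IsGradient, a horizontal RGB1-to-RGB2 gradient.
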Material /Game/Athena/UI/Frontend/Spatial/Art/M_Ui_Skewed_SpatialButton
graph {
 properties {
  bUsedWithStaticMesh=false
  blendMode=BLEND_Translucent
  materialDomain=MD_UI
  pixelDepthOffsetMode=PDOM_Legacy
  usedShadingModels=Default Lit
 }
 nodes [
  n1 = MaterialFunctionCall() {materialFunction=&GetUserInterfaceUV}
  n2 = Rotator(Coordinate=@n1.Normalized UV, Time=Divide(A=ScalarParameter() {defaultValue=-3, group=Degrees, parameterName=TopAngle}) {constB=360}) {centerX=1, speed=6.2832}
  n3 = LinearInterpolate(A=ScalarParameter() {defaultValue=0.125, parameterName=ReflectionOffset}, Alpha=ScalarParameter() {parameterName=Reflection anim}) {constB=0}
  n4 = Add(A=ScalarParameter() {defaultValue=-0.1, group=Offset, parameterName=TopOffset}, B=@n3)
  n5 = ScalarParameter() {defaultValue=128, group=Feather, parameterName=HorizontalSharpness}
  n6 = Multiply(A=ComponentMask(None=Add(A=@n2, B=@n4)) {g=true}, B=@n5)
  n7 = Rotator(Coordinate=@n1.Normalized UV, Time=Divide(A=ScalarParameter() {defaultValue=3, group=Degrees, parameterName=BottomAngle}) {constB=360}) {centerX=0, speed=6.2832}
  n8 = Add(A=@n7, B=Add(A=ScalarParameter() {defaultValue=0.1, group=Offset, parameterName=BottomOffset}, B=@n3))
  n9 = Multiply(A=OneMinus(None=ComponentMask(None=@n8) {g=true}), B=@n5)
  n10 = Saturate(None=Multiply(A=Saturate(None=@n6), B=Saturate(None=@n9)))
  n11 = Rotator(Coordinate=@n1.9-Slice UV, Time=Divide(A=ScalarParameter() {defaultValue=5, group=Degrees, parameterName=LeftAngle}) {constB=360}) {speed=6.2832}
  n12 = Add(A=ScalarParameter() {defaultValue=-0.1, group=Offset, parameterName=LeftOffset}, B=@n3)
  n13 = ScalarParameter() {defaultValue=32, group=Feather, parameterName=VerticalSharpness}
  n14 = Multiply(A=ComponentMask(None=Add(A=@n11, B=@n12)) {r=true}, B=@n13)
  n15 = Rotator(Coordinate=@n1.9-Slice UV, Time=Divide(A=ScalarParameter() {defaultValue=-5, group=Degrees, parameterName=RightAngle}) {constB=360}) {centerX=0, speed=6.2832}
  n16 = Add(A=@n15, B=Add(A=ScalarParameter() {defaultValue=0.1, group=Offset, parameterName=RightOffset}, B=@n3))
  n17 = Multiply(A=OneMinus(None=ComponentMask(None=@n16) {r=true}), B=@n13)
  n18 = Saturate(None=Multiply(A=Saturate(None=@n14), B=Saturate(None=@n17)))
  n19 = ScalarParameter() {defaultValue=0.5, parameterName=Buttonn_Opcaity, sliderMax=1}
  n20 = Saturate(None=Subtract(A=DotProduct(A=@n10, B=@n18), B=OneMinus(None=@n19)))
  n21 = ScalarParameter() {defaultValue=-0.1, group=Offset, parameterName=TopOffset}
  n22 = ScalarParameter() {defaultValue=128, group=Feather, parameterName=HorizontalSharpness}
  n23 = Multiply(A=ComponentMask(None=Add(A=@n2, B=@n21)) {g=true}, B=@n22)
  n24 = Add(A=@n7, B=ScalarParameter() {defaultValue=0.1, group=Offset, parameterName=BottomOffset})
  n25 = Multiply(A=OneMinus(None=ComponentMask(None=@n24) {g=true}), B=@n22)
  n26 = Saturate(None=Multiply(A=Saturate(None=@n23), B=Saturate(None=@n25)))
  n27 = ScalarParameter() {defaultValue=-0.1, group=Offset, parameterName=LeftOffset}
  n28 = ScalarParameter() {defaultValue=32, group=Feather, parameterName=VerticalSharpness}
  n29 = Multiply(A=ComponentMask(None=Add(A=@n11, B=@n27)) {r=true}, B=@n28)
  n30 = Add(A=@n15, B=ScalarParameter() {defaultValue=0.1, group=Offset, parameterName=RightOffset})
  n31 = Multiply(A=OneMinus(None=ComponentMask(None=@n30) {r=true}), B=@n28)
  n32 = Saturate(None=Multiply(A=Saturate(None=@n29), B=Saturate(None=@n31)))
  n33 = ScalarParameter() {defaultValue=0.25, parameterName=Reflection_Opcaity, sliderMax=1}
  n34 = Saturate(None=Subtract(A=DotProduct(A=@n26, B=@n32), B=OneMinus(None=@n33)))
  Add(A=@n20, B=@n34)
  n35 = VectorParameter() {group=Gradient, parameterName=RGB1, defaultValue=<rgba 0.0343 0.0467 0.0953 1>}
  n36 = TextureCoordinate()
  LinearInterpolate(
   A=@n35.RGB
   B=LinearInterpolate(
    A=@n35.RGB
    B=VectorParameter() {group=Gradient, parameterName=RGB2, defaultValue=<rgba 0.0782 0.1301 0.3712 1>}.RGB
    Alpha=ComponentMask(None=@n36) {r=true}
   )
   Alpha=ScalarParameter() {defaultValue=1, parameterName=IsGradient}
  )
  Multiply(A=@n20, B=@n34)
 ]
}
